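; Looping gameplay cue notify for the Rift Gauntlets dash that adds a physics wash force field to the dashing target.
; On application, after target validity and effects quality checks, it spawns FN_WashForce and attaches it to the target, reusing any existing instance to avoid duplicates.
; On removal, it sets a short lifespan on the field, animates its strength with a timeline, then destroys it.
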(event EventOnApplication(EitherVFX-preferthisone)
  (params (MyTarget :type Actor%20Object%20Reference) (Parameters :type Gameplay%20Cue%20Parameters%20Structure%20%28by%20ref%29) (ParticleComponents :type Array%20of%20FXSystem%20Component%20Object%20References) (AudioComponents :type Array%20of%20Audio%20Component%20Object%20References) (BurstCameraShakeInstance :type Camera%20Shake%20Base%20Object%20Reference) (BurstDecalInstance :type Decal%20Actor%20Object%20Reference))
  (|CallParentFunction MyTarget Parameters ParticleComponents AudioComponents BurstCameraShakeInstance BurstDecalInstance)
  (Utilities|IsValid MyTarget
    (:Is%20Valid
      (Utilities|EffectsQualityCheck 0
        (:Cinematic
          _)
        (:Epic
          _)
        (:High
          _))
      (Utilities|IsValid (Variables|Transient|GetPhysical%20Wash%20Force)
        (:Is%20Not%20Valid
          (bind _returnvalue (Game|SpawnActorfromClass MyTarget "/Game/Effects/Niagara/Blueprints/PhsyicsFields/FN_WashForce.FN_WashForce_C" (Math|Conversions|ToTransform(Vector) (Transformation|GetActorLocation MyTarget)) "AlwaysSpawn" "MultiplyWithRoot" 0 5e+02 1.0 -999.0 0))
          (bind _output_get (Variables|Transient|SetPhysical%20Wash%20Force _returnvalue))
          (Utilities|IsValid _output_get
            (:Is%20Valid
              (Transformation|AttachActorToActor _output_get MyTarget "None" "SnapToTarget" "KeepRelative" "KeepRelative" true))))) ; Don't spawn the Rotor Wash multiple times
    )))

(event EventOnRemoval(EitherVFX-preferthisversion)
  (params (MyTarget :type Actor%20Object%20Reference) (Parameters :type Gameplay%20Cue%20Parameters%20Structure%20%28by%20ref%29) (ParticleComponents :type Array%20of%20FXSystem%20Component%20Object%20References) (AudioComponents :type Array%20of%20Audio%20Component%20Object%20References) (BurstCameraShakeInstance :type Camera%20Shake%20Base%20Object%20Reference) (BurstDecalInstance :type Decal%20Actor%20Object%20Reference))
  (|CallParentFunction MyTarget Parameters ParticleComponents AudioComponents BurstCameraShakeInstance BurstDecalInstance)
  (bind _physical_wash_force (Variables|Transient|GetPhysical%20Wash%20Force)
    (:then
      (Actor|SetLifeSpan _physical_wash_force 1.0)
      (bind (_direction _strength) (Timeline|Timeline Play)
        (:Update
          (bind _physical_wash_force_2 (Variables|Transient|GetPhysical%20Wash%20Force)
            (:then
              (Variables|/Game/Effects/Niagara/Blueprints/PhsyicsFields/FN_WashForce.FN_WashForce_C|SetStrength _strength _physical_wash_force_2))
            (:else)))
        (:Finished
          (bind _physical_wash_force_3 (Variables|Transient|GetPhysical%20Wash%20Force)
            (:then
              (Actor|DestroyActor _physical_wash_force_3))
            (:else)))))
    (:else)))

(fn ConstructionScript
  (|CallParentFunction))
