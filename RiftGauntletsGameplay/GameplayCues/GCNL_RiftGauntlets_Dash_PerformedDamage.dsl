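; Looping gameplay cue for the Rift Gauntlets dash dealing damage, adding controller rumble and a physics push on application.
; After the parent logic, it plays force feedback on the target's local controller, picking right or left by whether the effect level is zero.
; When the physics field console variable is on, an effects quality check gates spawning a Niagara radial force field actor at the target.
; That actor is cached and spawned only if none is already valid.
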(event EventOnApplication(EitherVFX-preferthisone)
  (params (MyTarget :type Actor%20Object%20Reference) (Parameters :type Gameplay%20Cue%20Parameters%20Structure%20%28by%20ref%29) (ParticleComponents :type Array%20of%20FXSystem%20Component%20Object%20References) (AudioComponents :type Array%20of%20Audio%20Component%20Object%20References) (BurstCameraShakeInstance :type Camera%20Shake%20Base%20Object%20Reference) (BurstDecalInstance :type Decal%20Actor%20Object%20Reference))
  (bind (_normalizedmagnitude _rawmagnitude _effectcontext _matchedtagname _originaltag _aggregatedsourcetags _aggregatedtargettags _location _normal _instigator _effectcauser _sourceobject _physicalmaterial _gameplayeffectlevel) (Ability|GameplayCue|BreakGameplayCueParameters Parameters))
  (bind _returnvalue (Fortnite|GetFortPlayerControllerfromActor MyTarget))
  (|CallParentFunction MyTarget Parameters ParticleComponents AudioComponents BurstCameraShakeInstance BurstDecalInstance)
  (Utilities|IsValid MyTarget
    (:Is%20Valid
      (Utilities|FlowControl|Sequence
        (:then_0
          (Utilities|IsValid _returnvalue
            (:Is%20Valid
              (Game|Feedback|LocalPlayForceFeedback _returnvalue (select (== _gameplayeffectlevel 0) (Variables|Default|GetRightForceFeedback) (Variables|Default|GetLeftForceFeedback)) false "None"))))
        (:then_1
          (bind _returnvalue_2 (Development|GetConsoleVariableIntValue "r.PhysicsField.EnableField"))
          (if (== _returnvalue_2 1)
            (Utilities|EffectsQualityCheck 0
              (:Cinematic
                _)
              (:Epic
                _)
              (:High
                _))
            (Variables|Transient|GetRadialForce
              (:then)
              (:else
                (bind _returnvalue_3 (Game|SpawnActorfromClass self "/Game/Effects/Niagara/Blueprints/PhsyicsFields/FN_RadialForce.FN_RadialForce_C" (Math|Conversions|ToTransform(Vector) (Transformation|GetActorLocation MyTarget)) "AlwaysSpawn" "MultiplyWithRoot" 0 1.0 0.75 1e+03 2.0 0.0 0))
                (Variables|Transient|SetRadialForce _returnvalue_3)))))))))

(fn ConstructionScript
  (|CallParentFunction))
