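; Burst gameplay cue for the impact of a Rift Gauntlets side dash, adjusting where its impact VFX appears.
; After the parent burst, it takes a particle component and line-traces the effect context's hit component from the target along its normalized velocity.
; A successful trace moves the effect to the hit point.
; The effect is then rotated to face the target, even when no trace hit occurs.
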(fn OnBurst(EitherVFX-preferthisversion)
  (params (MyTarget :type Actor%20Object%20Reference) (Parameters :type Gameplay%20Cue%20Parameters%20Structure%20%28by%20ref%29) (ParticleComponents :type Array%20of%20FXSystem%20Component%20Object%20References) (AudioComponents :type Array%20of%20Audio%20Component%20Object%20References) (BurstCameraShakeInstance :type Camera%20Shake%20Base%20Object%20Reference) (BurstDecalInstance :type Decal%20Actor%20Object%20Reference))
  (bind (_normalizedmagnitude _rawmagnitude _effectcontext) (Ability|GameplayCue|BreakGameplayCueParameters Parameters))
  (bind _returnvalue (Ability|EffectContext|GetHitResult _effectcontext))
  (bind (_bblockinghit _binitialoverlap _time _distance _location _impactpoint _normal _impactnormal _physmat _hitactor _hitcomponent) (Collision|BreakHitResult _returnvalue))
  (bind _output (Utilities|Array|Get(acopy) ParticleComponents))
  (|CallParentFunction MyTarget Parameters ParticleComponents AudioComponents BurstCameraShakeInstance BurstDecalInstance)
  (Utilities|FlowControl|Sequence
    (:then_0
      (if (Utilities|Array|IsValidIndex ParticleComponents)
        (Utilities|IsValid _output
          (:Is%20Valid
            (Utilities|FlowControl|Sequence
              (:then_0
                (Utilities|IsValid MyTarget
                  (:Is%20Valid
                    (Utilities|IsValid _hitcomponent
                      (:Is%20Valid
                        (bind (_hitlocation _hitnormal _bonename _outhit _returnvalue_2) (Collision|LineTraceComponent _hitcomponent (Transformation|GetActorLocation MyTarget) (+ (Transformation|GetActorLocation MyTarget) (* (Math|Vector|Normalize (Transformation|GetVelocity MyTarget)) 2e+02)) true false false))
                        (if _returnvalue_2
                          (Transformation|SetWorldLocation _output _hitlocation false false)))))))
              (:then_1
                (Transformation|SetWorldRotation _output (Math|Conversions|RotationFromXVector (- (Transformation|GetActorLocation (Variables|GetMyTarget)) (Transformation|GetWorldLocation _output))) false false)))))))
    (:then_1
      (return))))
