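; Parent gameplay cue notify for Rift Gauntlets that moves impact effects nearer the part of the player dealing damage.
; GetAdjustedImpactPoint offsets a start from the pawn, finds the hit component's closest collision point, then line-traces that component toward it.
; If that fails, it returns the original hit point and normal.
; UpdateParticleComponentsToAdjustedImpactPoint applies the result to each particle component's location, rotation and NormalImpact parameter.
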(fn GetAdjustedImpactPoint
  (params (TargetActor :type Actor%20Object%20Reference) (Hit :type Hit%20Result%20Structure%20%28by%20ref%29))
  (bind (_bblockinghit _binitialoverlap _time _distance _location _impactpoint _normal _impactnormal _physmat _hitactor _hitcomponent) (Collision|BreakHitResult Hit))
  (bind _hitcomponent_2 (LocalVariables|GetHitComponent))
  ; AdjustedImpactPoint: Adjust the impact point so that it is closest to the bone/socket on the player that is actually doing the damage.
  (Utilities|FlowControl|Sequence
    (:then_0
      (bind _output_get (LocalVariables|SetHitComponent _hitcomponent))
      (Utilities|IsValid _output_get
        (:Is%20Valid
          (bind _asfort_player_pawn (Utilities|Casting|CastToFortPlayerPawn TargetActor))
          (bind _output_get_2 (LocalVariables|SetTraceStartLoc (Utilities|Operators|Add (* (.x (Variables|Default|GetHitTraceStartRelativeOffset)) (Transformation|GetActorForwardVector _asfort_player_pawn)) (* (Transformation|GetActorRightVector _asfort_player_pawn) (.y (Variables|Default|GetHitTraceStartRelativeOffset))) (* (Transformation|GetActorUpVector _asfort_player_pawn) (.z (Variables|Default|GetHitTraceStartRelativeOffset))) (Transformation|GetActorLocation _asfort_player_pawn))))
          (LocalVariables|SetClosestPointOnHitComponent (Collision|GetClosestPointonCollision _hitcomponent_2 _output_get_2 "None"))
          (if (>= (.ReturnValue (Collision|GetClosestPointonCollision _hitcomponent_2 _output_get_2 "None")) 0)
            (bind (_hitlocation _hitnormal _bonename _outhit _returnvalue) (Collision|LineTraceComponent (LocalVariables|GetHitComponent) (LocalVariables|GetTraceStartLoc) (+ (LocalVariables|GetTraceStartLoc) (* (Math|Vector|GetUnitDirection(Vector) (LocalVariables|GetTraceStartLoc) (LocalVariables|GetClosestPointOnHitComponent)) (Variables|Default|GetAdjustedImpactTraceDistance))) true false false))
            (if _returnvalue
              (return _hitlocation _hitnormal))))))
    (:then_1
      (return _impactpoint _impactnormal))))

(fn UpdateParticleComponentsToAdjustedImpactPoint
  (params (TargetActor :type Actor%20Object%20Reference) (ParticleComponents :type Array%20of%20FXSystem%20Component%20Object%20References) (GameplayCueParam :type Gameplay%20Cue%20Parameters%20Structure))
  (bind (_normalizedmagnitude _rawmagnitude _effectcontext) (Ability|GameplayCue|BreakGameplayCueParameters GameplayCueParam))
  ;[0.02092 0.06638 0.15 0.5] Update the spawned particles location and rotation.
  (Utilities|IsValid TargetActor
    (:Is%20Valid
      (bind (_impactpoint _impactnormal) (Default|GetAdjustedImpactPoint TargetActor (Ability|EffectContext|GetHitResult _effectcontext)))
      (for _array_element ParticleComponents
        (Utilities|IsValid _array_element
          (:Is%20Valid
            (Transformation|SetWorldLocationAndRotation _array_element _impactpoint (Math|Conversions|RotationFromXVector _impactnormal) false false)
            (Effects|Components|ParticleSystem|SetVectorParameter _array_element "NormalImpact" _impactnormal)))))))
