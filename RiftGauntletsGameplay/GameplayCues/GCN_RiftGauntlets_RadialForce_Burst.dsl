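; Burst gameplay cue for the Rift Gauntlets that spawns a radial-force physics field actor alongside the parent cue's burst behavior.
; It spawns only for a valid target when physics fields are enabled and effects quality passes a level check, read from a console variable in editor/PIE or from user settings otherwise.
; The field sits ahead of the target by the cue's raw magnitude, while the cue Location's components serve as field parameters rather than position.
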(fn IsValidEffectQuality
  (params (TargetActor :type Actor%20Object%20Reference))
  (bind _returnvalue (Settings|GetGameUserSettings))
  (if (Editor|InEditororPIE TargetActor)
    (bind _returnvalue_2 (Development|GetConsoleVariableIntValue "sg.EffectsQuality"))
    (LocalVariables|SetQualityLevel _returnvalue_2)
    (else
      (LocalVariables|SetQualityLevel (Settings|GetVisualEffectQuality _returnvalue))))
  (return (Math|Boolean|ORBoolean (== (LocalVariables|GetQualityLevel) 2) (== (LocalVariables|GetQualityLevel) 3) (== (LocalVariables|GetQualityLevel) 4))))

(fn OnBurst(EitherVFX-preferthisversion)
  (params (MyTarget :type Actor%20Object%20Reference) (Parameters :type Gameplay%20Cue%20Parameters%20Structure%20%28by%20ref%29) (ParticleComponents :type Array%20of%20FXSystem%20Component%20Object%20References) (AudioComponents :type Array%20of%20Audio%20Component%20Object%20References) (BurstCameraShakeInstance :type Camera%20Shake%20Base%20Object%20Reference) (BurstDecalInstance :type Decal%20Actor%20Object%20Reference))
  (bind (_normalizedmagnitude _rawmagnitude _effectcontext _matchedtagname _originaltag _aggregatedsourcetags _aggregatedtargettags _location) (Ability|GameplayCue|BreakGameplayCueParameters Parameters))
  (|CallParentFunction MyTarget Parameters ParticleComponents AudioComponents BurstCameraShakeInstance BurstDecalInstance)
  (Utilities|FlowControl|Sequence
    (:then_0
      (Utilities|IsValid MyTarget
        (:Is%20Valid
          (bind _returnvalue (Development|GetConsoleVariableIntValue "r.PhysicsField.EnableField"))
          (if (== _returnvalue 1)
            (bind _bvalid (CallFunction|IsValidEffectQuality MyTarget))
            (if _bvalid
              (Game|SpawnActorfromClass MyTarget "/Game/Effects/Niagara/Blueprints/PhsyicsFields/FN_RadialForce.FN_RadialForce_C" (Math|Conversions|ToTransform(Vector) (+ (Transformation|GetActorLocation MyTarget) (* (Transformation|GetActorForwardVector MyTarget) _rawmagnitude))) "AlwaysSpawn" "MultiplyWithRoot" 0 1.0 (.x _location) (.y _location) (.z _location) 0.0 0))))))
    (:then_1
      (return))))
